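# Ambient sound cue for the Chrono bomb pod plant, mixing two looping beds with intermittent one-shots.
# The main loop (pitched up) and the pulse loop each fade in through an Enveloper under a shared loop attenuation.
# A looping delayed branch plays pitch-varied random one-shots with separate attenuation; its quality switch uses eleven variants in the first two slots, three in the last two.
# Concurrency settings cap simultaneous instances.
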SoundCue /ChronoEnviroItems/LaunchablePlants/Gameplay/Assets/Sounds/Bomb_Plant/Chrono_PodPlant_Bomb_Ambient_Loop_Cue
root=Mixer_0
nodes [
 {
  id=Attenuation_0
  children [Enveloper_0]
  properties {attenuationSettings=/ChronoEnviroItems/LaunchablePlants/Gameplay/Assets/Sounds/Launchable_Plants_Ambient_Loop_Att}
 }
 {
  id=Attenuation_1
  children [Enveloper_1]
  properties {attenuationSettings=&Launchable_Plants_Ambient_Loop_Att}
 }
 {
  id=Attenuation_2
  children [Looping_0]
  properties {attenuationSettings=/ChronoEnviroItems/LaunchablePlants/Gameplay/Assets/Sounds/Launchable_Plants_Ambient_OneShot_Att}
 }
 {id=Delay_0, children [Modulator_2], properties {delayMax=6, delayMin=2}}
 {
  id=Enveloper_0
  children [Modulator_0]
  properties {pitchCurve {editorCurveData {keys [{time=0}]}}, volumeCurve {editorCurveData {keys [<at 0 0 RCIM_Cubic>,<at 1 1>]}}}
 }
 {
  id=Enveloper_1
  children [Modulator_1]
  properties {pitchCurve {editorCurveData {keys [{time=0}]}}, volumeCurve {editorCurveData {keys [<at 0 0 RCIM_Cubic>,<at 1 1>]}}}
 }
 {id=Looping_0, children [Delay_0]}
 {id=Mixer_0, children [Attenuation_0, Attenuation_1, Attenuation_2], properties {inputVolume [2,3.5,2]}}
 {id=Modulator_0, children [WavePlayer_0], properties {pitchMax=2, pitchMin=2}}
 {id=Modulator_1, children [WavePlayer_1], properties {pitchMax=1, pitchMin=1, volumeMax=1, volumeMin=1}}
 {id=Modulator_2, children [QualityLevel_0], properties {pitchMax=1.2, pitchMin=0.85}}
 {id=QualityLevel_0, children [Random_0,Random_0,Random_1,Random_1]}
 {
  id=Random_0
  children [
   WavePlayer_2
   WavePlayer_3
   WavePlayer_4
   WavePlayer_5
   WavePlayer_6
   WavePlayer_7
   WavePlayer_8
   WavePlayer_9
   WavePlayer_10
   WavePlayer_11
   WavePlayer_12
  ]
  properties {weights [1,1,1,1,1,1,1,1,1,1,1]}
 }
 {id=Random_1, children [WavePlayer_10, WavePlayer_11, WavePlayer_12], properties {bShouldExcludeFromBranchCulling=true, weights [1,1,1]}}
 {
  id=WavePlayer_0
  properties {bLooping=true, soundWaveAssetPtr=&/Chrono_PodPlant_Bomb_Ambient_Loop_01}
 }
 {
  id=WavePlayer_1
  properties {
   bLooping=true
   soundWaveAssetPtr=&/Chrono_PodPlant_Bomb_Ambient_Pulse_Loop
  }
 }
 {
  id=WavePlayer_10
  properties {soundWaveAssetPtr=&/PodPlant_Bomb_OneShots_09}
 }
 {
  id=WavePlayer_11
  properties {soundWaveAssetPtr=&/PodPlant_Bomb_OneShots_10}
 }
 {
  id=WavePlayer_12
  properties {soundWaveAssetPtr=&/PodPlant_Bomb_OneShots_11}
 }
 {
  id=WavePlayer_2
  properties {soundWaveAssetPtr=&/PodPlant_Bomb_OneShots_01}
 }
 {
  id=WavePlayer_3
  properties {soundWaveAssetPtr=&/PodPlant_Bomb_OneShots_02}
 }
 {
  id=WavePlayer_4
  properties {soundWaveAssetPtr=&/PodPlant_Bomb_OneShots_03}
 }
 {
  id=WavePlayer_5
  properties {soundWaveAssetPtr=&/PodPlant_Bomb_OneShots_04}
 }
 {
  id=WavePlayer_6
  properties {soundWaveAssetPtr=&/PodPlant_Bomb_OneShots_05}
 }
 {
  id=WavePlayer_7
  properties {soundWaveAssetPtr=&/PodPlant_Bomb_OneShots_06}
 }
 {
  id=WavePlayer_8
  properties {soundWaveAssetPtr=&/PodPlant_Bomb_OneShots_07}
 }
 {
  id=WavePlayer_9
  properties {soundWaveAssetPtr=&/PodPlant_Bomb_OneShots_08}
 }
]
properties {
 maxDistance=2000
 soundClassObject=&Fort_Ambient_Athena
 volumeMultiplier=1.253
 concurrencyOverrides {maxCount=3}
 concurrencySet [/ChronoEnviroItems/LaunchablePlants/Gameplay/Assets/Sounds/PlantPod_Loop_Conc]
}
