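# Sound cue for the bounce of the launchable Bomb pod plant, spatialized through the shared launchable-plant bounce attenuation.
# A Modulator at the root varies playback over a Quality Level switch that feeds random wave selection.
# The first two quality slots pick from all eight bounce waves, the last two from only three, in a random node excluded from branch culling.
# It plays in the impacts sound class and borrows rifle bullet impact concurrency.
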SoundCue /ChronoEnviroItems/LaunchablePlants/Gameplay/Assets/Sounds/Bomb_Plant/Chrono_PodPlant_Bomb_Bounce_Cue
root=Modulator_0
nodes [
 {id=Modulator_0, children [QualityLevel_0]}
 {id=QualityLevel_0, children [Random_0,Random_0,Random_1,Random_1]}
 {
  id=Random_0
  children [WavePlayer_0,WavePlayer_1,WavePlayer_4,WavePlayer_5,WavePlayer_6,WavePlayer_7,WavePlayer_8,WavePlayer_9]
  properties {weights [1,1,1,1,1,1,1,1]}
 }
 {id=Random_1, children [WavePlayer_7,WavePlayer_8,WavePlayer_9], properties {bShouldExcludeFromBranchCulling=true, weights [1,1,1]}}
 {
  id=WavePlayer_0
  properties {soundWaveAssetPtr=&/PodPlant_PodBounce_Bomb_01}
 }
 {
  id=WavePlayer_1
  properties {soundWaveAssetPtr=&/PodPlant_PodBounce_Bomb_02}
 }
 {
  id=WavePlayer_4
  properties {soundWaveAssetPtr=&/PodPlant_PodBounce_Bomb_03}
 }
 {
  id=WavePlayer_5
  properties {soundWaveAssetPtr=&/PodPlant_PodBounce_Bomb_04}
 }
 {
  id=WavePlayer_6
  properties {soundWaveAssetPtr=&/PodPlant_PodBounce_Bomb_05}
 }
 {
  id=WavePlayer_7
  properties {soundWaveAssetPtr=&/PodPlant_PodBounce_Bomb_06}
 }
 {
  id=WavePlayer_8
  properties {soundWaveAssetPtr=&/PodPlant_PodBounce_Bomb_07}
 }
 {
  id=WavePlayer_9
  properties {soundWaveAssetPtr=&/PodPlant_PodBounce_Bomb_08}
 }
]
properties {
 attenuationSettings=/ChronoEnviroItems/LaunchablePlants/Gameplay/Assets/Sounds/Launchable_Plants_Bounce_Att
 maxDistance=2500
 soundClassObject=&Fort_Impacts
 volumeMultiplier=2.5
 concurrencyOverrides {maxCount=4, resolutionRule=PreventNew}
 concurrencySet [&RifleBulletImpact_Conc]
}
