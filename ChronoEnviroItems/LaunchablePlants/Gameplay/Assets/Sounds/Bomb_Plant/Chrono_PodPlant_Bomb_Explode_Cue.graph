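# Sound cue for the Chrono bomb pod plant explosion, layering plant-specific sounds with generic weapon explosions.
# Separately attenuated close and distant plant layers use quality-level switches between eight-wave and three-wave random pools.
# A reduced-volume submix adds random air strike close, Quad Launcher medium and RPG distant explosions.
# These generic layers have pitch modulation and their own attenuations, with the medium and distant ones routed to a distant weapon-fire sound class.
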SoundCue /ChronoEnviroItems/LaunchablePlants/Gameplay/Assets/Sounds/Bomb_Plant/Chrono_PodPlant_Bomb_Explode_Cue
root=Mixer_0
nodes [
 {
  id=Attenuation_0
  children [QualityLevel_0]
  properties {attenuationSettings=/ChronoEnviroItems/LaunchablePlants/Gameplay/Assets/Sounds/Launchable_Plants_Explode_Att}
 }
 {
  id=Attenuation_1
  children [QualityLevel_1]
  properties {attenuationSettings=/ChronoEnviroItems/LaunchablePlants/Gameplay/Assets/Sounds/Launchable_Plants_Explode_Distant_Att}
 }
 {
  id=Attenuation_3
  children [Modulator_2]
  properties {
   attenuationSettings=/Game/Sounds/Attenuation/Fort_Explosion_Medium_Athena_Att
   attenuationOverrides {falloffDistance=3000, attenuationShapeExtents {x=1000}}
  }
 }
 {
  id=Attenuation_5
  children [Random_4]
  properties {attenuationSettings=&Launchable_Plants_Explode_Att}
 }
 {
  id=Attenuation_6
  children [Modulator_4]
  properties {attenuationSettings=&Launchable_Plants_Explode_Distant_Att}
 }
 {id=Mixer_0, children [Modulator_0,Modulator_1,Mixer_1], properties {inputVolume [1,1,0.5]}}
 {id=Mixer_1, children [SoundClass_1,SoundClass_0,SoundClass_4], properties {inputVolume [1,1,1]}}
 {id=Modulator_0, children [Attenuation_0]}
 {id=Modulator_1, children [Attenuation_1]}
 {id=Modulator_2, children [Random_6], properties {pitchMax=1.1, pitchMin=0.9}}
 {id=Modulator_3, children [Attenuation_5], properties {pitchMax=1.2}}
 {id=Modulator_4, children [Random_5], properties {pitchMax=1.1, pitchMin=0.9}}
 {id=QualityLevel_0, children [Random_0,Random_0,Random_2,Random_2]}
 {id=QualityLevel_1, children [Random_1,Random_1,Random_3,Random_3]}
 {
  id=Random_0
  children [WavePlayer_4, WavePlayer_5, WavePlayer_6, WavePlayer_7, WavePlayer_8, WavePlayer_9, WavePlayer_10, WavePlayer_11]
  properties {weights [1,1,1,1,1,1,1,1]}
 }
 {
  id=Random_1
  children [WavePlayer_12, WavePlayer_13, WavePlayer_14, WavePlayer_15, WavePlayer_16, WavePlayer_17, WavePlayer_18, WavePlayer_19]
  properties {weights [1,1,1,1,1,1,1,1]}
 }
 {id=Random_2, children [WavePlayer_9, WavePlayer_10, WavePlayer_11], properties {bShouldExcludeFromBranchCulling=true, weights [1,1,1]}}
 {id=Random_3, children [WavePlayer_17, WavePlayer_18, WavePlayer_19], properties {bShouldExcludeFromBranchCulling=true, weights [1,1,1]}}
 {id=Random_4, children [WavePlayer_22, WavePlayer_23, WavePlayer_24, WavePlayer_25], properties {weights [1,1,1,1]}}
 {id=Random_5, children [WavePlayer_0,WavePlayer_1,WavePlayer_2], properties {weights [1,1,1]}}
 {id=Random_6, children [WavePlayer_3, WavePlayer_20, WavePlayer_21], properties {weights [1,1,1]}}
 {
  id=SoundClass_0
  children [Attenuation_3]
  properties {soundClassOverride=&Fort_WeaponRanged_Fire_Distant}
 }
 {id=SoundClass_1, children [Modulator_3], properties {soundClassOverride=/Game/Sounds/FortSoundClassesAndModes/Fort_Explosives_Medium_PSM}}
 {
  id=SoundClass_4
  children [Attenuation_6]
  properties {soundClassOverride=&Fort_WeaponRanged_Fire_Distant}
 }
 {id=WavePlayer_0, properties {soundWaveAssetPtr=/Game/Athena/Sounds/Weapons/Explosions/RPG_Expl_Dist_01}}
 {id=WavePlayer_1, properties {soundWaveAssetPtr=/Game/Athena/Sounds/Weapons/Explosions/RPG_Expl_Dist_02}}
 {
  id=WavePlayer_10
  properties {soundWaveAssetPtr=&/PodPlant_PodExplode_Bomb_Close_07}
 }
 {
  id=WavePlayer_11
  properties {soundWaveAssetPtr=&/PodPlant_PodExplode_Bomb_Close_08}
 }
 {
  id=WavePlayer_12
  properties {soundWaveAssetPtr=&/PodPlant_PodExplode_Bomb_Distant_01}
 }
 {
  id=WavePlayer_13
  properties {soundWaveAssetPtr=&/PodPlant_PodExplode_Bomb_Distant_02}
 }
 {
  id=WavePlayer_14
  properties {soundWaveAssetPtr=&/PodPlant_PodExplode_Bomb_Distant_03}
 }
 {
  id=WavePlayer_15
  properties {soundWaveAssetPtr=&/PodPlant_PodExplode_Bomb_Distant_04}
 }
 {
  id=WavePlayer_16
  properties {soundWaveAssetPtr=&/PodPlant_PodExplode_Bomb_Distant_05}
 }
 {
  id=WavePlayer_17
  properties {soundWaveAssetPtr=&/PodPlant_PodExplode_Bomb_Distant_06}
 }
 {
  id=WavePlayer_18
  properties {soundWaveAssetPtr=&/PodPlant_PodExplode_Bomb_Distant_07}
 }
 {
  id=WavePlayer_19
  properties {soundWaveAssetPtr=&/PodPlant_PodExplode_Bomb_Distant_08}
 }
 {id=WavePlayer_2, properties {soundWaveAssetPtr=/Game/Athena/Sounds/Weapons/Explosions/RPG_Expl_Dist_03}}
 {id=WavePlayer_20, properties {soundWaveAssetPtr=/Game/Athena/Sounds/Weapons/Explosions/Quad_Launcher_Expl_Med_01}}
 {id=WavePlayer_21, properties {soundWaveAssetPtr=/Game/Athena/Sounds/Weapons/Explosions/Quad_Launcher_Expl_Med_02}}
 {id=WavePlayer_22, properties {soundWaveAssetPtr=/Game/Athena/Sounds/Weapons/Air_Strike/AppleSauce_Expl_Close_04}}
 {id=WavePlayer_23, properties {soundWaveAssetPtr=/Game/Athena/Sounds/Weapons/Air_Strike/AppleSauce_Expl_Close_01}}
 {id=WavePlayer_24, properties {soundWaveAssetPtr=/Game/Athena/Sounds/Weapons/Air_Strike/AppleSauce_Expl_Close_02}}
 {id=WavePlayer_25, properties {soundWaveAssetPtr=/Game/Athena/Sounds/Weapons/Air_Strike/AppleSauce_Expl_Close_03}}
 {id=WavePlayer_3, properties {soundWaveAssetPtr=/Game/Athena/Sounds/Weapons/Explosions/Quad_Launcher_Expl_Med_03}}
 {
  id=WavePlayer_4
  properties {soundWaveAssetPtr=&/PodPlant_PodExplode_Bomb_Close_01}
 }
 {
  id=WavePlayer_5
  properties {soundWaveAssetPtr=&/PodPlant_PodExplode_Bomb_Close_02}
 }
 {
  id=WavePlayer_6
  properties {soundWaveAssetPtr=&/PodPlant_PodExplode_Bomb_Close_03}
 }
 {
  id=WavePlayer_7
  properties {soundWaveAssetPtr=&/PodPlant_PodExplode_Bomb_Close_04}
 }
 {
  id=WavePlayer_8
  properties {soundWaveAssetPtr=&/PodPlant_PodExplode_Bomb_Close_05}
 }
 {
  id=WavePlayer_9
  properties {soundWaveAssetPtr=&/PodPlant_PodExplode_Bomb_Close_06}
 }
]
properties {
 maxDistance=12000
 soundClassObject=&Fort_Explosives_Medium_PSM
 volumeMultiplier=0.85
 concurrencyOverrides {maxCount=4, resolutionRule=PreventNew}
 concurrencySet [&RifleBulletImpact_Conc]
}
