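# Sound cue for the launched bomb pod plant, layering launch one-shots with a looping in-air sound.
# A mixer blends a briefly delayed loop that swells through a volume envelope, a heavily pitched-up random launch sound and a lower-pitched random pod bounce.
# A quality-level switch trims launch variations on some tiers.
# Layers use shared launchable-plant attenuations, and the cue overrides concurrency to cap simultaneous instances.
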SoundCue /ChronoEnviroItems/LaunchablePlants/Gameplay/Assets/Sounds/Bomb_Plant/Chrono_PodPlant_Bomb_InAirLoop_Cue
root=Mixer_0
nodes [
 {
  id=Attenuation_0
  children [Modulator_0]
  properties {attenuationSettings=/ChronoEnviroItems/LaunchablePlants/Gameplay/Assets/Sounds/Launchable_Plants_Launched_Att}
 }
 {
  id=Attenuation_1
  children [Modulator_1]
  properties {attenuationSettings=/ChronoEnviroItems/LaunchablePlants/Gameplay/Assets/Sounds/Launchable_Plants_InAir_Loop_Att}
 }
 {
  id=Attenuation_3
  children [Modulator_2]
  properties {attenuationSettings=&Launchable_Plants_Launched_Att}
 }
 {id=Delay_3, children [Attenuation_1], properties {delayMax=0.1, delayMin=0.1}}
 {
  id=Enveloper_0
  children [WavePlayer_0]
  properties {pitchCurve {editorCurveData {keys [{time=0}]}}, volumeCurve {editorCurveData {keys [<at 0 0.15 RCIM_Cubic>,<at 1 1>]}}}
 }
 {id=Mixer_0, children [Delay_3, Attenuation_0, Attenuation_3], properties {inputVolume [0.9,0.5,1]}}
 {id=Modulator_0, children [QualityLevel_0], properties {pitchMax=2.5, pitchMin=2.5, volumeMax=1}}
 {id=Modulator_1, children [Enveloper_0], properties {pitchMax=1, pitchMin=0.8, volumeMax=2, volumeMin=2}}
 {id=Modulator_2, children [Random_2], properties {pitchMax=0.9, pitchMin=0.8}}
 {id=QualityLevel_0, children [Random_0,Random_0,Random_1,Random_1]}
 {id=Random_0, children [WavePlayer_1,WavePlayer_2,WavePlayer_3,WavePlayer_4], properties {weights [1,1,1,1]}}
 {id=Random_1, children [WavePlayer_2,WavePlayer_3,WavePlayer_4], properties {bShouldExcludeFromBranchCulling=true, weights [1,1,1]}}
 {id=Random_2, children [WavePlayer_9, WavePlayer_10, WavePlayer_11], properties {weights [1,1,1]}}
 {
  id=WavePlayer_0
  properties {bLooping=true, soundWaveAssetPtr=&/PodPlant_InAir_Loop}
 }
 {
  id=WavePlayer_1
  properties {soundWaveAssetPtr=&/PodPlant_Launch_Bomb_01}
 }
 {
  id=WavePlayer_10
  properties {soundWaveAssetPtr=&/PodPlant_PodBounce_Bomb_01}
 }
 {
  id=WavePlayer_11
  properties {soundWaveAssetPtr=&/PodPlant_PodBounce_Bomb_02}
 }
 {
  id=WavePlayer_2
  properties {soundWaveAssetPtr=&/PodPlant_Launch_Bomb_02}
 }
 {
  id=WavePlayer_3
  properties {soundWaveAssetPtr=&/PodPlant_Launch_Bomb_03}
 }
 {
  id=WavePlayer_4
  properties {soundWaveAssetPtr=&/PodPlant_Launch_Bomb_04}
 }
 {
  id=WavePlayer_9
  properties {soundWaveAssetPtr=&/PodPlant_PodBounce_Bomb_03}
 }
]
properties {
 bOverrideConcurrency=true
 maxDistance=6000
 pitchMultiplier=1.5
 soundClassObject=&Fort_WeaponRanged_Projectile_Effects
 volumeMultiplier=2
 concurrencyOverrides {maxCount=3}
 concurrencySet [null]
}
