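# Sound cue for the launch of a bomb pod plant's pod, among the launchable plant environment items.
# A mixer layers a boosted, modulated random pick of three Bomb launch waves with a random pick of three launch waves borrowed from the Stink plant.
# It uses the shared launched-plant attenuation, the Fort_Impacts sound class and the rifle bullet impact concurrency set.
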SoundCue /ChronoEnviroItems/LaunchablePlants/Gameplay/Assets/Sounds/Bomb_Plant/Chrono_PodPlant_Bomb_Launched_Cue
root=Mixer_0
nodes [
 {id=Mixer_0, children [Modulator_0,Random_6], properties {inputVolume [3,1]}}
 {id=Modulator_0, children [Random_2]}
 {id=Random_2, children [WavePlayer_2,WavePlayer_3,WavePlayer_5], properties {weights [1,1,1]}}
 {id=Random_6, children [WavePlayer_0, WavePlayer_1, WavePlayer_10], properties {bShouldExcludeFromBranchCulling=true, weights [1,1,1]}}
 {
  id=WavePlayer_0
  properties {soundWaveAssetPtr=/ChronoEnviroItems/LaunchablePlants/Gameplay/Assets/Sounds/Stink_Plant/PodPlant_PodLaunch_Stink_01}
 }
 {
  id=WavePlayer_1
  properties {soundWaveAssetPtr=/ChronoEnviroItems/LaunchablePlants/Gameplay/Assets/Sounds/Stink_Plant/PodPlant_PodLaunch_Stink_02}
 }
 {
  id=WavePlayer_10
  properties {soundWaveAssetPtr=/ChronoEnviroItems/LaunchablePlants/Gameplay/Assets/Sounds/Stink_Plant/PodPlant_PodLaunch_Stink_08}
 }
 {
  id=WavePlayer_2
  properties {soundWaveAssetPtr=&/PodPlant_PodLaunch_Bomb_08}
 }
 {
  id=WavePlayer_3
  properties {soundWaveAssetPtr=&/PodPlant_PodLaunch_Bomb_01}
 }
 {
  id=WavePlayer_5
  properties {soundWaveAssetPtr=&/PodPlant_PodLaunch_Bomb_03}
 }
]
properties {
 attenuationSettings=/ChronoEnviroItems/LaunchablePlants/Gameplay/Assets/Sounds/Launchable_Plants_Launched_Att
 maxDistance=5008.436
 soundClassObject=&Fort_Impacts
 volumeMultiplier=3
 concurrencyOverrides {maxCount=4, resolutionRule=PreventNew}
 concurrencySet [&RifleBulletImpact_Conc]
}
