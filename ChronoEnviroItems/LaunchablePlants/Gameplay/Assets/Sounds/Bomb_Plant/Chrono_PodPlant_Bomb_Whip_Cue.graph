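# Sound cue for the bomb pod plant's whip effect, mixing a random whip one-shot with an in-air bomb-fall layer.
# A quality switch picks from five whip variants at the first two levels and three of them at the last two.
# The fall layer is pitched up before mixing, then a root modulator pitches the whole mix down.
# It plays in the ranged-projectile sound class with overridden spatial attenuation.
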SoundCue /ChronoEnviroItems/LaunchablePlants/Gameplay/Assets/Sounds/Bomb_Plant/Chrono_PodPlant_Bomb_Whip_Cue
root=Modulator_1
nodes [
 {id=Mixer_0, children [QualityLevel_0, Modulator_0], properties {inputVolume [1,1]}}
 {id=Modulator_0, children [WavePlayer_0], properties {pitchMax=2, pitchMin=2, volumeMax=1, volumeMin=1}}
 {id=Modulator_1, children [Mixer_0], properties {pitchMax=0.7, pitchMin=0.7}}
 {id=QualityLevel_0, children [Random_3,Random_3,Random_0,Random_0]}
 {id=Random_0, children [WavePlayer_3,WavePlayer_4,WavePlayer_5], properties {bShouldExcludeFromBranchCulling=true, weights [1,1,1]}}
 {id=Random_3, children [WavePlayer_1,WavePlayer_2,WavePlayer_3,WavePlayer_4,WavePlayer_5], properties {weights [1,1,1,1,1]}}
 {
  id=WavePlayer_0
  properties {soundWaveAssetPtr=&/PodPlant_InAir_BombFall}
 }
 {
  id=WavePlayer_1
  properties {soundWaveAssetPtr=&/PodPlant_Bomb_Whip_Highend_01}
 }
 {
  id=WavePlayer_2
  properties {soundWaveAssetPtr=&/PodPlant_Bomb_Whip_Highend_02}
 }
 {
  id=WavePlayer_3
  properties {soundWaveAssetPtr=&/PodPlant_Bomb_Whip_Highend_03}
 }
 {
  id=WavePlayer_4
  properties {soundWaveAssetPtr=&/PodPlant_Bomb_Whip_Highend_04}
 }
 {
  id=WavePlayer_5
  properties {soundWaveAssetPtr=&/PodPlant_Bomb_Whip_Highend_05}
 }
]
properties {
 attenuationSettings=/Game/Sounds/Attenuation/Fort_Weapon_Projectile_InAir_Loop_Att
 bDebug=true
 bOverrideAttenuation=true
 maxDistance=5512
 pitchMultiplier=0.9
 soundClassObject=&Fort_WeaponRanged_Projectile_Effects
 volumeMultiplier=0.7
 attenuationOverrides {
  bEnableReverbSend=false
  bUseComplexCollisionForOcclusion=true
  falloffDistance=5000
  occlusionVolumeAttenuation=0
  stereoSpread=0
  attenuationShapeExtents {x=512}
 }
}
